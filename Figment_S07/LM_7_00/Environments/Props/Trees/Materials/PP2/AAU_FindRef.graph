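# Editor utility Blueprint whose FindReferences function inspects what references the first selected asset, apparently as a work-in-progress debugging tool.
# The connected path gets that asset's package referencers, loads each one and runs inline Python to print its class property names.
# Unconnected leftovers include Asset Registry referencer and path queries, debug prints, a Material Instance Constant cast and an editor-property read of base property overrides.
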EditorUtilityBlueprint /Figment_S07/LM_7_00/Environments/Props/Trees/Materials/PP2/AAU_FindRef
graphs [
 EventGraph {links [], nodes []}
 FindReferences {nodes [
   n1 = GetSelectedAssets(execute=@n14.then)
   n2 = "Get (a copy)"(Array:Array of Object References=@n1) -> (Output:Object Reference)
   n3 = GetPathNameForLoadedAsset(LoadedAsset=@n2)
   > PrintString(InString=@n3)
   > n4 = Get Referencers!(
    execute:Exec
    self:Object Reference=@n2
    PackageName:Name=String To Name(InString=@n3)
    ReferenceOptions:"Asset Registry Dependency Options Structure (by ref)"=Make Asset Registry Dependency Options(bIncludeSoftPackageReferences=True)
   ) -> (then:Exec, OutReferencers:Array of Names, ReturnValue:bool) {is=title K2Node_Message}
   > n5 = For Each Loop(Array:Array of Names=@n4) -> (Array Element:Name)
   n6 = GetObjectName(Object=GetPackageForObject!(Object:Object Reference=@n2) -> (ReturnValue:Package Object Reference))
   n7 = FindPackageReferencersForAsset!(execute=@n1.then, AssetPath:String=@n6, bLoadAssetsToConfirm:bool=true) -> (then, ReturnValue:Array of Strings)
   > n8 = For Each Loop(Array:Array of Strings=@n7) -> (Array Element:String)
   PrintString(execute=@n5.LoopBody, InString=@n8)
   n9 = GetAssetRegistry()
   n10 = ParseIntoArray(SourceString=@n8, Delimiter=/)
   n11 = "Get (a copy)"(Array:Array of Strings=@n10, Dimension 1=Last Index!(TargetArray=@n10) -> (ReturnValue=0)) -> (Output:String)
   n12 = GetAssetsByPath!(
    execute
    self:Asset Registry Interface=@n9
    PackagePath:Name=String To Name(InString=Append!(A=@n6, B=., C:String=@n11) -> (ReturnValue))
    bRecursive:bool=false
    bIncludeOnlyOnDiskAssets:bool=false
   ) -> (then, OutAssetData:Array of Asset Data Structures, ReturnValue:bool)
   > n13 = For Each Loop(Array:Array of Asset Data Structures=@n12) -> (Array Element:Asset Data Structure)
   > PrintString(InString=Hello)
   n14 = FindReferences() {is=title K2Node_FunctionEntry}
   Break Asset Data(AssetData=@n13)
   Get Asset By Object Path!(
    execute
    self:Asset Registry Interface=@n9
    ObjectPath:"Soft Object Path Structure (by ref)"
    bIncludeOnlyOnDiskAssets:bool=false
    bSkipARFilteredAssets:bool=true
   ) -> (then, ReturnValue:Asset Data Structure) {is=title K2Node_CallFunction}
   String To Name(InString=GetPathNameForLoadedAsset())
   ExecutePythonCommand(PythonCommand="in = 1")
   "Execute Python Command (Advanced)"!(
    execute
    PythonCommand:String
    ExecutionMode:EPythonCommandExecutionMode Enum=ExecuteFile
    FileExecutionScope:EPythonFileExecutionScope Enum=Private
   ) -> (then, CommandResult:String, LogOutput:Array of Python Log Output Entry Structures, ReturnValue:bool)
   "To String (Name)"(InName=@n5)
   n15 = LoadAsset(execute=@n8.LoopBody, AssetPath=@n8)
   Cast To MaterialInstanceConstant(Object=@n15)
   Execute Python Script!(
    execute=@n15.then
    PythonScript="import unreal\r\n\r\nprops = unreal.PythonBPLib.get_all_property_names(object.static_class())\r\n\r\nprint(props)"
    object:Object Reference=@n15
   ) -> (then, ReturnValue) {inputs [object]}
   GetEditorProperty!(execute:Exec, Object:Object Reference=@n15, PropertyName:Name="base_property_overrides ") -> (then:Exec, PropertyValue:Wildcard, ReturnValue:bool)
  ]}
]
variables []
